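# Physics asset for the HumanBeing pickaxe mesh, defining secondary-motion simulation for a dangling hand, its pinky finger and a cord.
# Capsule bodies cover a hand base, hand, three pinky segments and three cord segments.
# The hand hinges off its base with moderate swing, and both chains branch from the hand.
# The pinky is nearly rigid with very tight limits, while cord joints loosen progressively toward the tip.
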PhysicsAsset /BRCosmetics/Weapons/FORT_Melee/Pickaxe_HumanBeing/Meshes/HumanBeing_Axe_PhysicsAsset
bodies {
 dyn_cord_01 [{shapeType=Capsule, length=2.3393, radius=0.6144, shapeName=dyn_cord_01_capsule, center=<1.3024 -0.1978 0>, rotation=<rot -90 180 -177.6245>}]
 dyn_cord_02 [{
   shapeType=Capsule
   length=2.3148
   radius=0.5885
   shapeName=dyn_cord_02_capsule
   center=<1.1739 0.1101 0>
   rotation=<rot -0.0009 88.3044 90.0013>
  }]
 dyn_cord_03 [{
   shapeType=Capsule
   length=0.9549
   radius=0.6224
   shapeName=dyn_cord_03_capsule
   center=<0.8663 0.1433 0.0357>
   rotation=<rot 0 95.4652 -82.8821>
  }]
 dyn_hand [{shapeType=Capsule, length=3, radius=3, shapeName=dyn_hand_capsule, rotation=<rot 0 0 90>}]
 dyn_hand_base [{shapeType=Capsule, length=0.505, radius=0.505, shapeName=dyn_hand_base_capsule, rotation=<rot 0 -90 0>}]
 dyn_pinky_01 [{
   shapeType=Capsule
   length=2.5904
   radius=1.5953
   shapeName=dyn_pinky_01_capsule
   center=<2.3653 0.0157 0.0266>
   rotation=<rot 0 90.3107 -86.6759>
  }]
 dyn_pinky_02 [{
   shapeType=Capsule
   length=1.7792
   radius=1.3086
   shapeName=dyn_pinky_02_capsule
   center=<1.8941 -0.0456 0.05>
   rotation=<rot 0 90.4907 -89.2896>
  }]
 dyn_pinky_03 [{
   shapeType=Capsule
   length=2.4591
   radius=1.2173
   shapeName=dyn_pinky_03_capsule
   center=<2.2171 -0.0818 -0.5461>
   rotation=<rot 0 -91.4453 -69.2987>
  }]
}
constraints [
 {bone1Name=dyn_pinky_01, bone2Name=dyn_hand, swing1LimitDegrees=0.1, swing2LimitDegrees=3, twistLimitDegrees=1}
 {bone1Name=dyn_pinky_02, bone2Name=dyn_pinky_01, swing1LimitDegrees=0.1, swing2LimitDegrees=4, twistLimitDegrees=1}
 {bone1Name=dyn_pinky_03, bone2Name=dyn_pinky_02, swing1LimitDegrees=0.1, swing2LimitDegrees=4, twistLimitDegrees=1}
 {bone1Name=dyn_hand, bone2Name=dyn_hand_base, swing1LimitDegrees=20, swing2LimitDegrees=20, twistLimitDegrees=2}
 {bone1Name=dyn_cord_03, bone2Name=dyn_cord_02, swing1LimitDegrees=14, swing2LimitDegrees=14, twistLimitDegrees=4}
 {bone1Name=dyn_cord_01, bone2Name=dyn_hand, swing1LimitDegrees=10, swing2LimitDegrees=10, twistLimitDegrees=2}
 {bone1Name=dyn_cord_02, bone2Name=dyn_cord_01, swing1LimitDegrees=12, swing2LimitDegrees=12, twistLimitDegrees=3}
]
